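; Gameplay ability for the Season 22 Octane flip quest, crediting players who flip the vehicle nearly two full rotations while airborne.
; On vehicle entry it checks vehicle tags and starts a looping timer; exiting clears it and ends the ability.
; Each tick accumulates absolute up-vector angle changes, signed around the right vector with 360-degree wrap correction, resetting on landing.
; Passing the threshold sends the quest stat event and ends the ability.
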(event Custom|Check%20In%20Air
  (bind _degrees_turned (Variables|Default|GetDegrees%20Turned))
  ;[1 1 1] Check if In the air
  (bind _asfort_athena_vehicle (Utilities|Casting|CastToFortAthenaVehicle (Vehicle|GetVehicleInterface (Variables|GAVars|Data|GetOwningPlayerPawn)))
    (:then
      (if (Vehicle|VehicleIsinAir (Vehicle|GetVehicleInterface (Variables|GAVars|Data|GetOwningPlayerPawn)))
        ;[1 1 1] If In the air start calculating rotation
        (bind _degrees_turned_2 (CallFunction|AngleBetweenVectors (Math|Vector|GetUpVector (Vehicle|GetVehicleRotation _asfort_athena_vehicle)) (Math|Vector|GetUpVector (Variables|Default|GetInitial%20rotation)) (Math|Vector|GetRightVector (Vehicle|GetVehicleRotation _asfort_athena_vehicle))))
        (bind _returnvalue (CallFunction|CalculateDegreesTurned (Variables|Default|GetLast%20degree%20reported) _degrees_turned_2))
        (if (< _returnvalue 0)
          (Variables|Default|SetDegrees%20Turned (+ (* _returnvalue -1.0) _degrees_turned))
          (else
            (Variables|Default|SetDegrees%20Turned (+ _degrees_turned _returnvalue))))
        (Variables|Default|SetLast%20degree%20reported _degrees_turned_2)
        (if (> (Variables|Default|GetDegrees%20Turned) 7.1e+02)
          (Quest|SendComplexCustomStatEvent (Variables|GAVars|Data|GetOwningPlayerQuestManager) 0 "(GameplayTags=((TagName=\"athena.quests.s22.Octane.Flip\")))" "(GameplayTags=)" 1)
          (Utilities|Time|ClearandInvalidateTimerbyHandle (Variables|Default|GetTimer%20Handle))
          (Ability|EndAbility))
        (else
          (Variables|Default|SetDegrees%20Turned 0.0)
          (Variables|Default|SetInitial%20rotation (Vehicle|GetVehicleRotation _asfort_athena_vehicle)))))
    (:CastFailed
      (Ability|EndAbility))))

(event Ability|EventActivateAbilityFromEvent
  (params (EventData :type Gameplay%20Event%20Data%20Structure%20%28by%20ref%29))
  (bind _timer_handle (Variables|Default|GetTimer%20Handle))
  (|CallParentFunction EventData)
  (switch Utilities|FlowControl|Switch|SwitchonGameplayTag (Utilities|Struct|BreakGameplayEventData EventData)
    (:Event.Vehicle.Entered
      ;[1 1 1] Check Vehicle Tags
      (if (GameplayTags|TagContainer|CompareTagAssetInterfacetoOtherTags (Variables|Default|GetT_Car) (Vehicle|GetVehicleAbilitySystemComponent (Vehicle|GetVehicleInterface (Variables|GAVars|Data|GetOwningPlayerPawn))))
        ;[1 1 1] Start Checking Vehicle If Vehicle is in Air
        (bind _returnvalue (Utilities|Time|SetTimerbyEvent (delegate Custom|Check%20In%20Air) (Variables|Default|GetExecutionTimeInterval) true))
        (Variables|Default|SetTimer%20Handle _returnvalue)
        (else
          (Ability|EndAbility))))
    (:Event.Vehicle.Exit
      ;[1 1 1] On Vehicle Exit Clear Timer and End Ability
      (if (Utilities|Time|IsValidTimerHandle _timer_handle)
        (Utilities|Time|ClearandInvalidateTimerbyHandle _timer_handle)
        (else
          _))
      (Ability|EndAbility))))

(fn AngleBetweenVectors
  (params (Destination%20Rotation :type Vector) (Origin%20Rotation :type Vector) (Rotation%20Normal :type Vector))
  (return (Math|Float|SelectFloat (Math|Trig|RadiansToDegrees (Math|Trig|Acos(Radians) (/ (Math|Vector|DotProduct Origin%20Rotation Destination%20Rotation) (* (Math|Vector|VectorLength Destination%20Rotation) (Math|Vector|VectorLength Origin%20Rotation))))) (- 3.6e+02 (Math|Trig|RadiansToDegrees (Math|Trig|Acos(Radians) (/ (Math|Vector|DotProduct Origin%20Rotation Destination%20Rotation) (* (Math|Vector|VectorLength Destination%20Rotation) (Math|Vector|VectorLength Origin%20Rotation)))))) (> (Math|Vector|DotProduct Rotation%20Normal (Math|Vector|CrossProduct Origin%20Rotation Destination%20Rotation)) 0))))

(fn CalculateDegreesTurned
  (params (A :type Float%20%28double-precision%29) (B :type Float%20%28double-precision%29))
  (Variables|Default|SetValue%20A A)
  (Variables|Default|SetValue%20B B)
  (if (and (< A 9e+01) (> B 2.7e+02))
    (Variables|Default|SetValue%20A (+ (Variables|Default|GetValue%20A) 3.6e+02))
    (elif (and (> A 2.7e+02) (< B 9e+01))
      (Variables|Default|SetValue%20B (+ (Variables|Default|GetValue%20B) 3.6e+02))
      (else
        _)))
  (return (- (Variables|Default|GetValue%20A) (Variables|Default|GetValue%20B))))
